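; Mesh component that acts as a fake Lumen emissive light, driven by Temp, Brightness and LowerBrightness variables.
; When override is enabled, MakeConfigurable creates a dynamic instance of M_FakeLumenLight on slot 0 and sets those three scalar parameters.
; pushValuesToPrimData instead writes the values as custom primitive data, all three to index 0.
; BeginPlay and Tick are empty, and neither function is called within this graph.
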(event EventBeginPlay)

(event EventTick
  (params (DeltaSeconds :type Float%20%28single-precision%29)))

(fn MakeConfigurable
  (if (Variables|Default|GetOverrideWithCustomValues?)
    (bind _returnvalue (Rendering|Material|CreateDynamicMaterialInstance 0 "/Game/Global/Materials/Lighting/M_FakeLumenLight.M_FakeLumenLight"))
    (Rendering|Material|SetMaterial 0 _returnvalue)
    (Rendering|Material|SetScalarParameterValue _returnvalue "Temp" (Variables|Default|GetTemp) :!class /Script/Engine.MaterialInstanceDynamic)
    (Rendering|Material|SetScalarParameterValue _returnvalue "Brightness" (Variables|Default|GetBrightness) :!class /Script/Engine.MaterialInstanceDynamic)
    (Rendering|Material|SetScalarParameterValue _returnvalue "LowerBrightness" (Variables|Default|GetLowerBrightness) :!class /Script/Engine.MaterialInstanceDynamic)))

(fn pushValuesToPrimData
  (Rendering|Material|SetCustomPrimitiveDataFloat 0 (Variables|Default|GetTemp))
  (Rendering|Material|SetCustomPrimitiveDataFloat 0 (Variables|Default|GetBrightness))
  (Rendering|Material|SetCustomPrimitiveDataFloat 0 (Variables|Default|GetLowerBrightness)))
